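# Conversation database for the NanaSplit supply specialist NPC (Peely Bone), who greets players with wordless action descriptions.
# The entry point leads to one speech node with a default line and requirement-gated variants for first-ever meetings, match conversation count and banana-character outfits.
# First-meeting and outfit reactions take priority from named curve rows.
# The speech node then branches through link nodes to the shared NPC Library quests and basic services conversations.
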ConversationDatabase /NPCLibrary/NPCs/NanaSplit/NPCLibrary_Conversation_NanaSplit
entry_points [{tag=Conversation.Entry.NanaSplit, destinations [#1]}]
nodes [
 ConversationEntryPointNode_7 {
  id=#1
  class=ConversationEntryPointNode
  outputs [#2]
  properties {entryTag {tagName=Conversation.Entry.NanaSplit}}
 }
 Speech_C_4 {
  id=#2
  class=Speech_C
  outputs [#3, #4]
  properties {
   nodeName=""
   generalConfig {
    defaultMessage=(Peely Bone taps on his ribs absently.)
    contextualMessages [
     {
      message=(Peely Bone waves happily.)
      contextRequirements [{requirement {class=FortControllerRequirement_IsFirstEverConversationWithNPC}}]
      priority {curve {rowName=Tandem.Speech.Priority.FirstEverInteraction}}
     }
     {
      message=(Peely Bone's stomach gurgles visibly.)
      contextRequirements [{requirement {
         class=FortControllerRequirement_CheckNPCMatchConversationCount
         properties {nPCUniqueIdTag {tagName=AISpawnerData.Type.Specialist.Supply.NanaSplit}, requiredConversationCountConfig {requirement=GreaterThan}}
        }}]
      priority {curve {curveTable=null, rowName=None}}
     }
     {
      message=(Peely Bone taps on his ribs absently.)
      contextRequirements [{requirement {
         class=FortControllerRequirement_CheckNPCMatchConversationCount
         properties {nPCUniqueIdTag {tagName=AISpawnerData.Type.Specialist.Supply.NanaSplit}, requiredConversationCountConfig {requirement=GreaterThan}}
        }}]
      priority {curve {curveTable=null, rowName=None}}
     }
     {
      message=(Peely Bone scratches an itch on his head-ribs.)
      contextRequirements [{requirement {
         class=FortControllerRequirement_CheckNPCMatchConversationCount
         properties {nPCUniqueIdTag {tagName=AISpawnerData.Type.Specialist.Supply.NanaSplit}, requiredConversationCountConfig {requirement=GreaterThan}}
        }}]
      priority {curve {curveTable=null, rowName=None}}
     }
     {
      message=(Peely Bone searches intently for your ribs.)
      contextRequirements [{requirement {
         class=FortControllerRequirement_HasCID
         properties {allowedCIds [
           &CID_349_Athena_Commando_M_Banana
           /BRCosmetics/Athena/Items/Cosmetics/Characters/Character_BananaCake
           &Character_BananaAdventure
           &CID_A_323_Athena_Commando_M_BananaWinter
           &CID_814_Athena_Commando_M_BananaSummer
          ]}
        }}]
      priority {curve {rowName=Tandem.Speech.Priority.OutfitReaction}}
     }
    ]
   }
  }
 }
 ConversationLinkNode_0 {id=#4, class=ConversationLinkNode, properties {remoteEntryTag {tagName=Conversation.Entry.BasicServices}}}
 ConversationLinkNode_1 {id=#3, class=ConversationLinkNode, properties {remoteEntryTag {tagName=Conversation.Entry.NPCLibrary.Quests}}}
]
